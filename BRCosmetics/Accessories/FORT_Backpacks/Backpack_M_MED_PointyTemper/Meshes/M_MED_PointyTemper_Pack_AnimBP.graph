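# Anim Blueprint for the PointyTemper backpack, adding physics jiggle and ambient floating motion to its ball.
# AnimGraph offsets the pack bone, simulates dyn_ball with spring-based AnimDynamics, then modifies rot_ball translation, with slow cubic bool blending, and rotation through cached poses.
# EventGraph computes time-driven rotation, layered-sine translation and a standing-still flag each update.
# It also sets a gender-based offset on customization load and changes velocity/acceleration scales on jump start and fall end.
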AnimBlueprint /BRCosmetics/Accessories/FORT_Backpacks/Backpack_M_MED_PointyTemper/Meshes/M_MED_PointyTemper_Pack_AnimBP
target_skeleton=&/M_MED_PointyTemper_Pack_Skeleton
graphs [
 AnimGraph {nodes [
   Output Pose(Result=Use cached pose 'etherealMovementCP'()) {layerGroup=None}
   n1 = "Transform (Modify) Bone - Bone: pack"!(ComponentPose=Local To Component(LocalPose=Input Pose()), bAlphaBoolEnabled:bool, Translation="0,-0.55,0") -> (Pose) {
    alphaInputType=Bool
    rotationSpace=BCS_WorldSpace
    scaleSpace=BCS_WorldSpace
    translationMode=BMM_Additive
    translationSpace=BCS_BoneSpace
    alphaBoolBlend {blendOption=Linear}
    boneToModify {boneName=pack}
    translation {y=-0.55}
   }
   n2 = Anim Dynamics - Bone: dyn_ball!(ComponentPose=@n1, Alpha=0.8, ComponentLinearAccScale:Vector="0,0,0", ComponentLinearVelScale:Vector="0,0,0") -> (Pose) {
    angularDampingOverride=1
    angularSpringConstant=20000
    bLinearSpring=true
    bOverrideLinearDamping=true
    bUseGravityOverride=true
    gravityScale=0.35
    linearDampingOverride=0.99
    linearSpringConstant=50
    simulationSpace=RootRelative
    alphaBoolBlend {blendOption=Linear}
    boundBone {boneName=dyn_ball}
    componentAppliedLinearAccClamp=<3000 3000 3000>
    componentLinearAccScale=<0.1 0.05 0.1>
    componentLinearVelScale=<0.1 0.05 0.1>
    physicsBodyDefinitions [{
      boundBone {boneName=dyn_ball}
      boxExtents=<50 50 50>
      constraintSetup {angularLimitsMax=<0 0 0>, angularLimitsMin=<0 0 0>, angularTarget=<1 0 0>, linearAxesMax=<5 0 8>, linearAxesMin=<-5 -5 -8>}
      localJointOffset=<5 0 0>
     }]
   }
   Save cached pose 'inputCP'(Pose=Component To Local(ComponentPose=@n2)) {cacheName=inputCP}
   n3 = "Transform (Modify) Bone - Bone: rot_ball"!(ComponentPose=Local To Component(LocalPose=Use cached pose 'inputCP'()), bAlphaBoolEnabled:bool=True, Translation="0,0,0") -> (Pose) {
    alphaInputType=Bool
    rotationSpace=BCS_WorldSpace
    translationMode=BMM_Additive
    translationSpace=BCS_BoneSpace
    alphaBoolBlend {blendInTime=4, blendOption=Cubic, blendOutTime=4}
    boneToModify {boneName=rot_ball}
   }
   n4 = "Transform (Modify) Bone - Bone: rot_ball"!(ComponentPose=@n3, Alpha:float=1, Rotation:Rotator="0,0,0") -> (Pose) {
    rotationMode=BMM_Replace
    rotationSpace=BCS_BoneSpace
    translationSpace=BCS_WorldSpace
    alphaBoolBlend {blendOption=Linear}
    boneToModify {boneName=rot_ball}
   }
   Save cached pose 'etherealMovementCP'(Pose=Component To Local(ComponentPose=@n4)) {cacheName=etherealMovementCP}
  ]}
 EventGraph {nodes [
   Event BlueprintUpdateAnimation()
   > n1 = Sequence!(execute) -> (then_0, then_1, then_2:Exec, then_3:Exec)
   n2 = float * float(A=float / float(A=GetRealTimeSeconds(), B=-720) {is=title K2Node_CallFunction}, B=10000)
   > Set Ambient Rotation(Ambient Rotation:Rotator=MakeRotator(Roll=@n2, Yaw=@n2)) -> (Output_Get:Rotator)
   n3 = GetRealTimeSeconds()
   n4 = float * float(A="Sin (Radians)"(A=float * float(A=1.2, B=@n3)), B=0.15)
   n5 = float * float(A="Sin (Radians)"(A=float * float(A=1, B=@n3)), B=0.3)
   n6 = Remap Float(Input="Sin (Radians)"(A=float * float(A=1.4, B=@n3)), OldRange=(X=-1,Y=1), NewRange=(X=-1,Y=2)) -> (Output=~)
   n7 = float * float(A="Clamp (Float)"(Value=@n6, Min=-1, Max=0), B=0.15)
   Set Ambient Translation(execute=@n1.then_1, Ambient Translation:Vector=vector + vector(A=MakeVector(X=@n4, Y=@n5, Z=@n6), B=MakeVector(Y=@n7))) -> (Output_Get:Vector)
   n8 = Get bIsMoving2D()
   n9 = Get bIsFalling()
   n10 = Get bIsJumping()
   Set isStandingStill(execute=@n1.then_2, isStandingStill:bool=AND Boolean(A=NOT Boolean(A=@n8), B=NOR Boolean!(A:bool=@n9, B:bool=@n10) -> (ReturnValue:bool))) -> (Output_Get:bool)
   Event OnCharacterCustomizationLoaded()
   > n11 = Branch(Condition="Equal (Enum)"(
     A:EFortCustomGender Enum=GetOwnerGender!(self:Custom Character Part Anim Instance Object Reference) -> (ReturnValue:EFortCustomGender Enum=Invalid)
     B:EFortCustomGender Enum=Male
    ))
   > Set packMoveGenderBased(packMoveGenderBased:Vector="0,0.5,0") -> (Output_Get:Vector)
   Set packMoveGenderBased(execute=@n11.else, packMoveGenderBased="0,0.4,0")
   Event OnBeginJumping()
   > Set Component Linear Vel Scale(Component Linear Vel Scale:Vector="0.1,0.05,0.05") -> (Output_Get:Vector)
   > Set Component Linear Acc Scale(Component Linear Acc Scale:Vector="0.1,0.05,0.03") -> (Output_Get:Vector)
   Event OnEndFalling()
   > Set Component Linear Vel Scale(Component Linear Vel Scale="0.1,0.05,0.1")
   > Set Component Linear Acc Scale(Component Linear Acc Scale="0.1,0.05,0.1")
  ]}
]
variables [
 Ambient Rotation {type=Rotator}
 Ambient Translation {type=Vector}
 Component Linear Acc Scale {type=Vector, default=<0.1 0.05 0.1>}
 Component Linear Vel Scale {type=Vector, default=<0.1 0.05 0.1>}
 isStandingStill {type=boolean}
 packMoveGenderBased {type=Vector}
]
